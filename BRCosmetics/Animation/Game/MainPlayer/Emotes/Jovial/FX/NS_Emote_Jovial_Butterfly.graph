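# Niagara effect for the Jovial player emote that spawns mesh butterflies in a single-emitter burst, each living four seconds.
# Particles start on a seeded cylinder surface with sideways velocity, scaling in and out over their lifetime.
# A waveform bobs their position, while late-surging local Y acceleration and fading drag send them flying off.
# Dynamic material parameters feed the mesh material, one driven by normalized system loop age.
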NiagaraSystem /BRCosmetics/Animation/Game/MainPlayer/Emotes/Jovial/FX/NS_Emote_Jovial_Butterfly
dependencies {usedDataInterfaces [&NiagaraDataInterfaceCurve], usedDynamicInputs [], usedModules [&SystemState]}
emitters {Butterfly {
  emitterSpawnScript [/SetVariables(Emitter.CameraFade=10) {isSetParameters=true}]
  emitterUpdateScript [
   &EmitterState(Inactive Response=::NewEnumerator1, Life Cycle Mode=::NewEnumerator1, Loop Behavior=::NewEnumerator1, Loop Duration=4, UseLoopDelay=false)
   &SpawnBurst_Instantaneous(Use Spawn Probability=false, UseLoopCountLimit=false)
  ]
  particleSpawnScript [
   &InitializeParticle(
    Color=<rgba 0.2779 0.1859 1 1>
    Lifetime=4
    Mesh Scale Mode=::NewEnumerator3
    Mesh Uniform Scale=0.04
    UsePositionOffset=false
    Write Material Random=true
   )
   &InitialMeshOrientation(Mesh Axis To Orient=<0 0 1>, Mesh Orientation Mode=::NewEnumerator2, Use Rotation Vector=false)
   &ShapeLocation(
    Cylinder Height=50
    Cylinder Radius=90
    MaskSpawnGroup=false
    Offset=<-20 -60 135>
    Offset Mode=::NewEnumerator0
    Override Seed=true
    Random Seed=7
    Randomness Mode:ENiagaraRandomnessMode=::NewEnumerator1
    Shape Primitive=::NewEnumerator1
    Surface Only=true
    Surface Only Band Thickness=1
    Use Endcaps In Surface Only Mode=false
    UseOwnerScale=true
   )
   &AddVelocity(Velocity=<0 1 0>, Velocity Speed Scale=54) {module=AddVelocity001}
  ]
  particleUpdateScript [
   &ParticleState()
   &ScaleMeshSize(Scale Factor=&VectorFromFloat(Value=&FloatFromCurve(FloatCurve=DataInterface() {propertyValues {Curve {keys [
          {interpMode=RCIM_Cubic, tangentMode=RCTM_User, time=0, value=0}
          <at 0.25 1>
          {interpMode=RCIM_Cubic, tangentMode=RCTM_User, time=0.8, value=1}
          {tangentMode=RCTM_User, time=1, value=0}
         ]}}}))) {module=ScaleMeshSize001}
   &DynamicMaterialParameters(
    Index 0 Param 1=0.25
    Index 0 Param 4=&Multiply_Float(A=$System.NormalizedLoopAge, B=2)
    Index 1 Param 1=7000
    Index 1 Param 2=1
    Module.Param0WriteEnabled=true
    Module.Param0WriteEnabled1=true
    Module.Param1WriteEnabled=true
    Module.Param1WriteEnabled1=true
    Module.Param2WriteEnabled=true
    Module.Param2WriteEnabled1=true
    Module.Param3WriteEnabled=true
    Module.Param3WriteEnabled1=true
    Write Parameter Index 1=true
   ) {module=DynamicMaterialParameters001}
   /SetVariables(Particles.Position=&AddVectorToPosition(
     Position=$Particles.Position
     Vector=&Multiply_VectorByFloat(
      Vector=<0 1 1>
      Float=&Waveform("Waveform [1]"=::NewEnumerator0, "[1] Frequency"=1, "[1] Amplitude Min/Max"=<-1 1>)
     )
    )) {isSetParameters=true}
   &AccelerationForce(
    Acceleration=&Multiply_VectorByFloat(
     Vector=<0 1 0>
     Float=&FloatFromCurve(FloatCurve=DataInterface() {propertyValues {Curve {keys [
          <at 0 1>
          {interpMode=RCIM_Cubic, leaveTangent=2803.9333, tangentMode=RCTM_Break, time=0.5, value=1}
          {interpMode=RCIM_Cubic, tangentMode=RCTM_Break, time=1, value=333}
         ]}}})
    )
    Coordinate Space=::Local
   )
   &Drag(
    Drag=&FloatFromCurve(FloatCurve=DataInterface() {propertyValues {Curve {keys [
         {arriveTangent=-4.2199, interpMode=RCIM_Cubic, leaveTangent=-4.2199, tangentMode=RCTM_User, time=0, value=1}
         {interpMode=RCIM_Cubic, tangentMode=RCTM_User, time=1, value=0}
        ]}}})
    Use Linear Drag=true
    Use Rotational Drag=false
   ) {module=Drag001}
   &SolveForcesAndVelocity(Clamp Velocity=false, Limit Acceleration=false)
  ]
  rendererClasses [&NiagaraMeshRendererProperties]
  renderers [{rendererClass=&NiagaraMeshRendererProperties, rendererIndex=0}]
 }}
summary {
 systemName=NS_Emote_Jovial_Butterfly
 emitters [{emitterName=Butterfly, rendererClasses [&NiagaraMeshRendererProperties]}]
 userVariables []
}
